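# Material function in the Homebase environment folder that outputs a squash-and-stretch World Position Offset.
# It subtracts world position, by default excluding shader offsets, from the component pivot, then splits that vector with axis masks.
# A scalar amount scales both parts linearly: positive values pull Z toward the pivot and push X and Y outward, negative values reverse it.
# Inputs allow overriding world position and pivot, so other deformations can be chained.
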MaterialFunction /Game/Environments/Homebase/MaterialFunctions/SquashAndStretch
graph {
 properties {
  bExposeToLibrary=true
  description="Scales the model outward on x and y as it decreases z and vice versa. "
  libraryCategoriesText [Misc, World Position Offset]
 }
 nodes [
  n1 = FunctionInput() {
   bUsePreviewValueAsDefault=true
   description=If there are any deformations that occur before this node add them to world position before inserting them into this input
   inputName=Squash and Stretch Amount
   inputType=FunctionInput_Scalar
  }
  n2 = FunctionInput(Preview=WorldPosition() {worldPositionShaderOffset=WPT_ExcludeAllShaderOffsets}.XYZ) {
   bUsePreviewValueAsDefault=true
   description=If there are any deformations that occur before this node add them to world position before inserting them into this input.
   inputName=World position
   sortPriority=1
  }
  n3 = Subtract(
   A=FunctionInput(Preview=MaterialFunctionCall() {materialFunction=&ComponentPivotLocation}.Result) {bUsePreviewValueAsDefault=true, description=Usually not needed, inputName=Component Pivot Loc, sortPriority=2}
   B=@n2
  )
  n4 = Multiply(A=Constant3Vector() {constant=<rgba 0 0 1 0>}, B=@n3)
  n5 = Multiply(A=@n3, B=Constant3Vector() {constant=<rgba -1 -1 0 0>})
  FunctionOutput(None=Add(A=Multiply(A=@n1, B=@n4) {constB=-1}, B=Multiply(A=@n1, B=@n5) {constB=-1})) {description=Add other distortions to this if they take place before or after this operation., outputName=WPO}
 ]
}
