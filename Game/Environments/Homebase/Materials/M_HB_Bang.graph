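# Lit material for the Homebase Bang mesh, also usable on mesh particles, that makes it spin, bob and glow.
# World position offset math combines Pivot Painter per-object Z rotation, a sine vertical bob and SquashAndStretch, with a function correcting rotated normals.
# A two-color UV gradient scaled by Hover forms the color.
# Emissive adds an inverse-Fresnel core glow to a flash term; BuildFlash and Fade modulate both color and glow.
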Material /Game/Environments/Homebase/Materials/M_HB_Bang
graph {
 properties {
  bDisableDepthTest=true
  bUsedWithMeshParticles=true
  numCustomizedUVs=3
  pixelDepthOffsetMode=PDOM_Legacy
  previewMesh=/Game/Environments/Homebase/Meshes/SM_HB_Bang
  translucencyLightingMode=TLM_VolumetricDirectional
  translucencyPass=MTP_BeforeDOF
  usedShadingModels=Default Lit
 }
 nodes [
  n1 = MaterialFunctionCall() {materialFunction=/Engine/Functions/Engine_MaterialFunctions02/PivotPainter/PivotPainter_PerObjectData}
  n2 = Constant3Vector() {constant=<rgba 0 0 1 0>}
  n3 = Time()
  n4 = Multiply(A=ScalarParameter() {defaultValue=0.5, parameterName=rotationSpeed}, B=@n3)
  n5 = WorldPosition()
  n6 = Constant2Vector()
  n7 = ComponentMask(None=MaterialFunctionCall!(Sine Phase=@n3, Value 1=Constant() {r=-16}, Value 2=Constant() {r=16}) {materialFunction=/Engine/Functions/Engine_MaterialFunctions02/Utility/Sine_Remapped}.Remapped Sine) {r=true}
  Add(
   A=MaterialFunctionCall!(Squash and Stretch Amount=ScalarParameter() {parameterName=Squash and Stretch}, World position, Component Pivot Loc=@n1.Pivot Position) {materialFunction=/Game/Environments/Homebase/MaterialFunctions/SquashAndStretch}.WPO
   B=Add(
    A=RotateAboutAxis(NormalizedRotationAxis=@n2, RotationAngle=@n4, PivotPoint=@n1.Pivot Position, Position=@n5.XYZ)
    B=AppendVector(A=@n6, B=@n7)
   )
  )
  ComponentMask(None=MaterialFunctionCall(Sine Phase=Time(), Value 1=Constant() {r=-0.8}, Value 2=Constant() {r=0.8}) {materialFunction=&Sine_Remapped}.Remapped Sine) {r=true}
  Constant() {r=0.6}
  n8 = LinearInterpolate(
   A=VectorParameter() {parameterName=Color1, defaultValue=<rgba 1 0.3877 0 1>}.RGB
   B=VectorParameter() {parameterName=Color2, defaultValue=<rgba 0.465 0.1119 0 1>}.RGB
   Alpha=ComponentMask(None=TextureCoordinate()) {g=true}
  )
  n9 = ScalarParameter() {defaultValue=1, parameterName=Hover}
  n10 = ScalarParameter() {defaultValue=1, parameterName=Fade}
  LinearInterpolate(
   A=Constant()
   B=LinearInterpolate(A=Multiply(A=@n8, B=@n9), B=Constant() {r=1}, Alpha=ScalarParameter() {parameterName=BuildFlash})
   Alpha=@n10
  )
  n11 = VectorParameter() {parameterName=CoreGlowColor, defaultValue=<rgba 1 0.4428 0 2>}
  n12 = Fresnel() {exponent=0.5}
  n13 = Multiply(A=Multiply(A=@n11.RGB, B=@n11.RGB), B=Power(Base=OneMinus(None=@n12)) {constExponent=0.75})
  n14 = LinearInterpolate(
   A=Multiply(A=VectorParameter() {parameterName=HBFlashColor, defaultValue=<rgba 1 1 1 0>}.RGB, B=ScalarParameter() {parameterName=FlashMult}) {constB=20}
   B=Constant() {r=3}
   Alpha=ScalarParameter() {parameterName=BuildFlash}
  )
  MaterialFunctionCall(Emissive=LinearInterpolate(A=Constant(), B=Add(A=@n13, B=@n14), Alpha=@n10)) {materialFunction=&MF_FX_EmissiveAdjustments}
  MaterialFunctionCall(Rotation Axis=@n2, Rotation Angle=@n4) {materialFunction=&FixRotateAboutAxisNormals}
  MaterialFunctionCall() {materialFunction=/Game/Environments/Homebase/MaterialFunctions/MF_HB_SmallWorld}
  ScalarParameter() {parameterName=Metallic}
  ScalarParameter() {defaultValue=0.66, parameterName=Roughness}
 ]
}
