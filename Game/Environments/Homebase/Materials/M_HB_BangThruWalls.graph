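# Translucent Homebase mesh-particle material for a "bang" marker that renders through walls by disabling depth test.
# Time-driven rotation around the particle's Z axis plus a sine vertical bob builds a position offset, with a companion function fixing rotated normals.
# A UV-gradient blend of two colors accompanies an inverse-Fresnel core glow, scaled by Fade and fed through emissive adjustments.
# It also includes the Homebase SmallWorld function.
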Material /Game/Environments/Homebase/Materials/M_HB_BangThruWalls
graph {
 properties {
  bDisableDepthTest=true
  bUsedWithMeshParticles=true
  blendMode=BLEND_Translucent
  numCustomizedUVs=3
  pixelDepthOffsetMode=PDOM_Legacy
  translucencyLightingMode=TLM_VolumetricPerVertexDirectional
  translucencyPass=MTP_BeforeDOF
  usedShadingModels=Default Lit
 }
 nodes [
  n1 = Constant3Vector() {constant=<rgba 0 0 1 0>}
  n2 = Time()
  n3 = Multiply(A=ScalarParameter() {defaultValue=0.5, parameterName=rotationSpeed}, B=@n2)
  n4 = MaterialFunctionCall!(Sine Phase=@n2, Value 1=Constant() {r=-16}, Value 2=Constant() {r=16}) {materialFunction=/Engine/Functions/Engine_MaterialFunctions02/Utility/Sine_Remapped}
  Add(
   A=RotateAboutAxis(NormalizedRotationAxis=@n1, RotationAngle=@n3, PivotPoint=ParticlePositionWS(), Position=WorldPosition().XYZ)
   B=AppendVector(A=Constant2Vector(), B=ComponentMask(None=@n4.Remapped Sine) {r=true})
  )
  LinearInterpolate(
   A=VectorParameter() {parameterName=Color1, defaultValue=<rgba 1 0.0417 0.0417 1>}.RGB
   B=VectorParameter() {parameterName=Color2, defaultValue=<rgba 0.915 0.1481 0.0294 1>}.RGB
   Alpha=ComponentMask(None=TextureCoordinate()) {g=true}
  )
  n5 = VectorParameter() {parameterName=CoreGlowColor, defaultValue=<rgba 1 0.0167 0 2>}
  n6 = Multiply(A=@n5.RGB, B=@n5.RGB)
  n7 = Power(Base=OneMinus(None=Fresnel() {exponent=0.5})) {constExponent=0.75}
  MaterialFunctionCall(Emissive=LinearInterpolate(A=Constant(), B=Multiply(A=@n6, B=@n7), Alpha=ScalarParameter() {defaultValue=1, parameterName=Fade})) {materialFunction=&MF_FX_EmissiveAdjustments}
  MaterialFunctionCall(Rotation Axis=@n1, Rotation Angle=@n3) {materialFunction=&FixRotateAboutAxisNormals}
  MaterialFunctionCall() {materialFunction=/Game/Environments/Homebase/MaterialFunctions/MF_HB_SmallWorld}
  ScalarParameter() {parameterName=Metallic}
  ScalarParameter() {defaultValue=0.66, parameterName=Roughness}
  ScalarParameter() {defaultValue=0.6, parameterName=Specular}
 ]
}
