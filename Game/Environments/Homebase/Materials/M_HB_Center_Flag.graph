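# Two-sided Default Lit material for the Homebase center flag, combining a tintable banner pattern with animated cloth-like motion.
# A Pattern texture on UV1 shades the Color parameter and masks a lighter icon, feeding a desaturated Fresnel highlight and a subsurface-style term.
# Panned swirl-noise normals, weighted by vertex-color red, build a world-position-offset vector and a two-layer ripple normal.
# Material output links are absent, and BuildFlash, HomeBaseParameterCollection FlagColor and debug nodes form separate branches.
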Material /Game/Environments/Homebase/Materials/M_HB_Center_Flag
graph {
 properties {
  pixelDepthOffsetMode=PDOM_Legacy
  previewMesh=/SaveTheWorld/Environments/Homebase/Meshes/SM_HB_Center_Flag_Repositioned_noLOD
  twoSided=true
  usedShadingModels=Default Lit
 }
 nodes [
  n1 = VertexNormalWS()
  n2 = PixelNormalWS()
  n3 = ScalarParameter() {defaultValue=0.3, parameterName=FresnelNormalBlend}
  n4 = Multiply(
   A=Fresnel(ExponentIn=ScalarParameter() {defaultValue=2, parameterName=FresnelLightPower}, Normal=LinearInterpolate(A=@n1, B=@n2, Alpha=@n3)) {baseReflectFraction=0}
   B=ScalarParameter() {defaultValue=1, parameterName=FresnelLightAmount}
  )
  n5 = VectorParameter() {desc=Color of the homebase flag, parameterName=Color, defaultValue=<rgba 0.6 0.0364 0.0184 0>}
  n6 = TextureSampleParameter2D(UVs=TextureCoordinate() {coordinateIndex=1, desc="-1,1"}) {parameterName=Pattern, texture=/Game/UI/HomeBase/BannerWidget/Patterns/HB-Banner-Pattern-Ironclad}
  n7 = LinearInterpolate(A=Multiply(A=Constant() {r=0.5}, B=@n5.RGB), B=@n5.RGB, Alpha=Power(Base=OneMinus(None=@n6.A)) {constExponent=50})
  n8 = ScalarParameter() {defaultValue=-0.25, parameterName=IconMaskALerp}
  n9 = ScalarParameter() {defaultValue=4, parameterName=IconMaskBLerp}
  n10 = LinearInterpolate(A=@n7, B=ScalarParameter() {defaultValue=0.9, parameterName=IconLightHigh}, Alpha=Clamp(None=LinearInterpolate(A=@n8, B=@n9, Alpha=@n6.A)))
  Clamp(None=Desaturation(None=Multiply(A=@n4, B=@n10), Fraction=ScalarParameter() {defaultValue=0.25, parameterName=FresnelSaturation}))
  ComponentMask(None=CollectionParameter() {collection=&/HomeBaseParameterCollection, parameterName=FlagColor}) {b=true, g=true, r=true}
  Constant3Vector() {constant=<rgba -0.5 -0.5 1 1>}
  Constant3Vector() {constant=<rgba 30 15 0 1>}
  Constant3Vector() {constant=<rgba -1 -1 1 0>}
  n11 = Add()
  n12 = Clamp(None=TwoSidedSign()) {clampMode=CMODE_ClampMin}
  LinearInterpolate(A=@n11, B=@n11, Alpha=@n12)
  n13 = ScalarParameter() {parameterName=BuildFlash}
  LinearInterpolate(A=Constant(), B=Constant() {r=1}, Alpha=@n13)
  LinearInterpolate(B=Constant() {r=1}, Alpha=@n13)
  n14 = TextureCoordinate() {vTiling=1.5}
  n15 = ScalarParameter() {defaultValue=0.3, parameterName=Scale}
  n16 = Time()
  n17 = ScalarParameter() {defaultValue=2, parameterName=Speed}
  n18 = TextureSample(UVs=Panner(Coordinate=Multiply(A=@n14, B=@n15), Time=Multiply(A=@n16, B=@n17)) {speedX=0.1, speedY=0.025}) {samplerType=SAMPLERTYPE_Normal, texture=/Game/Global/Textures/Default/Noise/T_SwirlNoise_N}
  n19 = Constant3Vector() {constant=<rgba -1 -1 1 0>}
  n20 = VertexColor()
  LinearInterpolate(
   A=Constant3Vector() {constant=<rgba 0 0 1 0>}
   B=Multiply(A=Multiply(A=@n18.RGB, B=@n19), B=VectorParameter() {parameterName=WPOPower, defaultValue=<rgba 30 10 5 1>}.RGB)
   Alpha=Power(Base=@n20.R, Exp=Constant() {r=3})
  )
  LinearInterpolate(A=Constant3Vector() {constant=<rgba 1 0 0 0>}, B=Constant3Vector() {constant=<rgba 0 1 0 0>}, Alpha=@n12)
  Multiply(A=Clamp(None=@n7), B=ScalarParameter() {defaultValue=0.35, parameterName=SSAmount})
  n21 = MaterialFunctionCall(Float2=TextureCoordinate() {coordinateIndex=1, desc="1,1"}) {materialFunction=&BreakOutFloat2Components}
  n22 = Constant() {r=0.5}
  Multiply(
   A=Multiply(B=Constant() {r=2})
   B=Multiply(A=SphereMask(A=@n21.R, B=@n22) {attenuationRadius=0.5}, B=SphereMask(A=@n21.G, B=@n22) {attenuationRadius=0.5})
  )
  n23 = Panner(
   Coordinate=Multiply(A=TextureCoordinate() {uTiling=0.5}, B=ScalarParameter() {defaultValue=1.5, parameterName=ScaleB})
   Time=Multiply(A=@n16, B=ScalarParameter() {defaultValue=1.25, parameterName=SpeedB})
  ) {speedX=0.1, speedY=-0.05}
  n24 = MaterialFunctionCall(
   BaseNormal=@n18.RGB
   AdditionalNormal=Multiply(
    A=TextureSample(UVs=@n23) {samplerType=SAMPLERTYPE_Normal, texture=&T_SwirlNoise_N}.RGB
    B=VectorParameter() {parameterName=SmallNormalsPower, defaultValue=<rgba 0.5 0.5 1 1>}.RGB
   ) {constB=-1}
  ) {materialFunction=&BlendAngleCorrectedNormals}
  n25 = Constant3Vector() {constant=<rgba -1 -1 1 0>}
  Multiply(
   A=VectorParameter() {parameterName=MasterNormalPower, defaultValue=<rgba 1 1 0.5 1>}.RGB
   B=LinearInterpolate(A=Constant3Vector() {constant=<rgba 0 0 1 0>}, B=Multiply(A=@n24.Result, B=@n25), Alpha=@n20.R)
  )
  ScalarParameter() {defaultValue=0.8, parameterName=Roughness}
  ScalarParameter() {defaultValue=0.9, parameterName=Opacity}
  ScalarParameter() {defaultValue=0.2, parameterName=Spec}
  TextureSample() {texture=/Game/UI/HomeBase/BannerWidget/Patterns/HB-Banner-Null}
  TextureSample() {texture=&HB-Banner-Pattern-Ironclad}
  VectorParameter() {parameterName=DEBUGCOLORA, defaultValue=<rgba 0.38 0.0075 0 1>}
 ]
}
