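# Material function that blends grass and soil ground surfaces for Homebase environments, outputting Diffuse, Normal, Roughness and Mask.
# World-space XY divided by a collection texture size drives both textures, with an extra scale on the soil.
# Vertex green blends dead and live grass tints from the parameter collection, and blue blends tinted grass into tinted soil.
# Red passes out as Mask; the normal is world up in tangent space, and roughness is one scalar.
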MaterialFunction /Game/Environments/Homebase/MaterialFunctions/MF_HB_GroundBlend
graph {nodes [
  FunctionOutput(None=ScalarParameter() {defaultValue=0.8, parameterName=RoughnessGrass_MF}) {outputName=Roughness}
  n1 = VertexColor()
  n2 = LinearInterpolate(
   A=ComponentMask(None=CollectionParameter() {collection=/Game/Environments/Homebase/Materials/HomeBaseParameterCollection, parameterName=GrassColorDead}) {b=true, g=true, r=true}
   B=ComponentMask(None=CollectionParameter() {collection=&HomeBaseParameterCollection, parameterName=GrassColor}) {b=true, g=true, r=true}
   Alpha=@n1.G
  )
  n3 = Divide(
   A=ComponentMask(None=WorldPosition() {worldPositionShaderOffset=WPT_ExcludeAllShaderOffsets}.XYZ) {g=true, r=true}
   B=ComponentMask(None=CollectionParameter() {collection=&HomeBaseParameterCollection, parameterName=GrassTextureSize}) {r=true}
  )
  n4 = TextureSampleParameter2D(UVs=@n3) {group=Blend, parameterName=BlendTexture, texture=/Game/Environments/Homebase/Textures/T_HB_Grass_D}
  n5 = TextureSampleParameter2D(UVs=Multiply(A=ScalarParameter() {defaultValue=1, parameterName=Blend2Scale}, B=@n3)) {parameterName=BlendTexture2, texture=/Game/Environments/Homebase/Textures/T_HB_Soil_D}
  n6 = ComponentMask(None=CollectionParameter() {collection=&HomeBaseParameterCollection, parameterName=SoilColor}) {b=true, g=true, r=true}
  FunctionOutput(None=LinearInterpolate(A=Multiply(A=@n2, B=@n4.RGB), B=Multiply(A=@n5.RGB, B=@n6), Alpha=@n1.B)) {outputName=Diffuse}
  FunctionOutput(None=Transform(None=Constant3Vector() {constant=<rgba 0 0 1 0>}) {transformSourceType=TRANSFORMSOURCE_World, transformType=TRANSFORM_Tangent}) {outputName=Normal}
  FunctionOutput(None=VertexColor().R) {outputName=Mask}
 ]}
