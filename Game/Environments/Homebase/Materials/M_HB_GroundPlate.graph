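# Lit ground plate material for the Homebase center ground, with a tinted glow at night.
# Center ground diffuse, normal and texture-driven roughness are lerped with MF_HB_GroundBlend outputs.
# Emissive multiplies the blended diffuse by a mask blue channel and tint, with nested Ifs on collection time of day hard-switching it off between two daytime borders.
# The separate day and night brightness parameters are unconnected.
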Material /Game/Environments/Homebase/Materials/M_HB_GroundPlate
graph {
 properties {pixelDepthOffsetMode=PDOM_Legacy, usedShadingModels=Default Lit}
 nodes [
  Constant() {r=0.2}
  n1 = MaterialFunctionCall() {materialFunction=/Game/Environments/Homebase/MaterialFunctions/MF_HB_GroundBlend}
  LinearInterpolate(
   A=TextureSample() {samplerType=SAMPLERTYPE_Normal, texture=/Game/Environments/Homebase/Textures/T_HB_Center_Ground_N}.RGB
   B=@n1.Normal
   Alpha=@n1.Normal
  )
  n2 = TextureSample() {texture=/Game/Environments/Homebase/Textures/T_HB_Center_Ground_D}
  LinearInterpolate(
   A=LinearInterpolate(A=ScalarParameter() {defaultValue=0.9292, parameterName=Roughness1}, B=ScalarParameter() {parameterName=Roughness2}, Alpha=@n2.B)
   B=@n1.Roughness
   Alpha=@n1.Roughness
  )
  n3 = LinearInterpolate(A=@n2.RGB, B=@n1.Diffuse, Alpha=@n1.Diffuse)
  n4 = TextureSample() {texture=/Game/Environments/Homebase/Textures/T_HB_Center_Ground_M}
  n5 = CollectionParameter() {collection=&FortniteMaterialParameters, parameterName=TimeOfDay_0-1}
  n6 = ScalarParameter() {defaultValue=0.8417, group=Emissive, parameterName=NightDayBorder}
  n7 = Constant()
  n8 = Constant() {r=1}
  n9 = If(A=@n5, B=ScalarParameter() {defaultValue=0.2712, group=Emissive, parameterName=DayNightBorder}, "A > B"=@n8, "A == B"=@n8, "A < B"=@n7)
  Multiply(
   A=Multiply(A=Multiply(A=@n3, B=@n4.B), B=VectorParameter() {parameterName=EmissiveTint, defaultValue=<rgba 1 0.7439 0.1919 1>}.RGB)
   B=LinearInterpolate(
    A=ScalarParameter() {defaultValue=4.3386, parameterName=EmissiveBrightness}
    B=Constant()
    Alpha=If(A=@n5, B=@n6, "A > B"=@n7, "A == B"=@n7, "A < B"=@n9)
   )
  )
  ScalarParameter() {defaultValue=0.5, group=Emissive, parameterName=EmissiveBrightnessNight}
  ScalarParameter() {defaultValue=0.5, group=Emissive, parameterName=EmissiveBrightnessDay}
 ]
}
