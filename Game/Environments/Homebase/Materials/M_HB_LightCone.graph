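# M_HB_LightCone is an unlit additive material for Homebase light cones that changes brightness between day and night.
# Emissive color scales by a brightness picked from the global TimeOfDay collection parameter, with a sine-based flicker masked by red vertex color.
# Opacity falls off along UV V through a power curve and depth fade.
# A SquashAndStretch function deforms it around either a manual or Pivot Painter pivot, chosen by static switch.
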Material /Game/Environments/Homebase/Materials/M_HB_LightCone
graph {
 properties {blendMode=BLEND_Additive, pixelDepthOffsetMode=PDOM_Legacy, shadingModel=MSM_Unlit, usedShadingModels=Unlit}
 nodes [
  n1 = ComponentMask(None=TextureCoordinate()) {g=true}
  DepthFade(Opacity=Power(Base=OneMinus(None=@n1), Exp=Constant() {r=3}), FadeDistance=ScalarParameter() {defaultValue=30, parameterName=DepthFade})
  n2 = VectorParameter() {parameterName=ManualPivot, defaultValue=<rgba 0 0 0 1>}
  MaterialFunctionCall!(
   Squash and Stretch Amount=ScalarParameter() {parameterName=Squash and Stretch}
   World position
   Component Pivot Loc=StaticSwitchParameter(
    True=TransformPosition(None=@n2.RGB) {transformType=TRANSFORMPOSSOURCE_World}
    False=MaterialFunctionCall() {materialFunction=/Engine/Functions/Engine_MaterialFunctions02/PivotPainter/PivotPainter_PerObjectData}.Pivot Position
   ) {defaultValue=true, group=PivotPositions, parameterName=ManualPivost}
  ) {materialFunction=/Game/Environments/Homebase/MaterialFunctions/SquashAndStretch}
  n3 = CollectionParameter() {collection=&FortniteMaterialParameters, parameterName=TimeOfDay_0-1}
  n4 = Constant()
  n5 = ScalarParameter() {defaultValue=0.25, group=Emissive, parameterName=DayNightBorder}
  n6 = Constant() {r=1}
  n7 = LinearInterpolate(
   A=ScalarParameter() {defaultValue=2, parameterName=BrightnessNight}
   B=ScalarParameter() {parameterName=BrightnessDay}
   Alpha=If(
    A=@n3
    B=ScalarParameter() {defaultValue=0.75, group=Emissive, parameterName=NightDayBorder}
    "A > B"=@n4
    "A == B"=@n4
    "A < B"=If(A=@n3, B=@n5, "A > B"=@n6, "A == B"=@n6, "A < B"=@n4)
   )
  )
  n8 = Multiply(A=Time(), B=ScalarParameter() {defaultValue=1, group=Emissive, parameterName=FlickerSpeed})
  n9 = Multiply(A=@n8) {constB=0.33}
  n10 = Add(A=Abs(None=Sine(None=@n9)), B=Multiply(A=Sine(None=@n8), B=ScalarParameter() {defaultValue=1, group=Emissive, parameterName=FlickerPause}))
  n11 = OneMinus(None=Frac(None=Clamp(None=@n10) {minDefault=-1}))
  Multiply(
   A=VectorParameter() {parameterName=EmissiveColor, defaultValue=<rgba 0.995 0.995 0.995 1>}.RGB
   B=LinearInterpolate(A=@n7, B=Multiply(A=@n7, B=@n11), Alpha=VertexColor().R)
  )
  ScalarParameter() {defaultValue=0.2, group=Emissive, parameterName=EmissiveFlicker}
  ScalarParameter() {defaultValue=0.5, group=Emissive, parameterName=EmissiveBrightnessNight}
  ScalarParameter() {defaultValue=0.5, group=Emissive, parameterName=EmissiveBrightnessDay}
 ]
}
